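# EQS query that picks a long-range position for an AI bot relative to its current target, using token positions or a new spot.
# The first option takes the target's token provider positions tagged Long range, favoring those nearest the bot and filtering by leash.
# The fallback option generates donut points through a composite generator, adding a height-offset obstruction test against the target to the same distance and leash tests.
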EnvQuery /NPCCore/EQS/EQS_LongRange_TokenPosOrNewPos
options [
 {
  generator {
   name=FortQueryGenerator_TokenProviderPositions_0
   class=FortQueryGenerator_TokenProviderPositions
   properties {
    bAutoSortTests=true
    bCanRunAsync=false
    envQueryResultNormalizationOption=Default
    optionName=""
    tokenProviderContext=&FortAthenaAIBot_EQSQueryContext_CurrentTarget
    gameplayTagQuery {
     autoDescription=" ANY( Athena.AI.Token.Range.Long )"
     tokenStreamVersion=0
     userDescription=""
     queryTokenStream [0,1,1,1,0]
     tagDictionary [{tagName=Athena.AI.Token.Range.Long}]
    }
   }
  }
  tests [
   EnvQueryTest_Distance_0 {
    class=EnvQueryTest_Distance
    properties {
     distanceTo=&EnvQueryContext_Querier
     filterType=Range
     floatValueMin=0
     scoringEquation=Linear
     scoringFactor=-1
     testComment=""
     testMode=Distance3D
     testPurpose=Score
     referenceValue {dataBinding=null, dataField=FloatValue, defaultValue=0}
    }
   }
   FortQueryTest_InsideAIBotLeash_0 {class=FortQueryTest_InsideAIBotLeash, properties {testPurpose=FilterAndScore}}
  ]
 }
 {
  generator {
   name=EnvQueryGenerator_Composite_1
   class=EnvQueryGenerator_Composite
   properties {optionName="", generators [&self:EnvQueryGenerator_Composite_1.EnvQueryGenerator_Donut_0]}
  }
  tests [
   EnvQueryTest_Distance_2 {
    class=EnvQueryTest_Distance
    properties {
     distanceTo=&EnvQueryContext_Querier
     filterType=Range
     floatValueMin=0
     scoringEquation=Linear
     scoringFactor=-1
     testComment=""
     testMode=Distance3D
     testPurpose=Score
     referenceValue {dataBinding=null, dataField=FloatValue, defaultValue=0}
    }
   }
   FortQueryTest_IsObstructed_1 {
    class=FortQueryTest_IsObstructed
    properties {
     context=&FortAthenaAIBot_EQSQueryContext_CurrentTarget
     contextHeightOffset=150
     itemHeightOffset=150
     testPurpose=FilterAndScore
    }
   }
   FortQueryTest_InsideAIBotLeash_1 {class=FortQueryTest_InsideAIBotLeash, properties {testPurpose=FilterAndScore}}
  ]
 }
]
