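# EQS query that picks melee-range attack token positions around an AI bot's current target.
# A token provider generator gathers positions from the current target, keeping those tagged Athena.AI.Token.Range.Melee.
# A negatively weighted linear 3D distance test scores positions nearer the querier higher, and a leash test both filters and scores by whether they lie inside the bot's leash.
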EnvQuery /NPCCore/EQS/EQS_MeleeRange_TokenPos
options [{
  generator {
   name=FortQueryGenerator_TokenProviderPositions_0
   class=FortQueryGenerator_TokenProviderPositions
   properties {
    bAutoSortTests=true
    bCanRunAsync=false
    envQueryResultNormalizationOption=Default
    optionName=""
    tokenProviderContext=&FortAthenaAIBot_EQSQueryContext_CurrentTarget
    gameplayTagQuery {
     autoDescription=" ANY( Athena.AI.Token.Range.Melee )"
     tokenStreamVersion=0
     userDescription=""
     queryTokenStream [0,1,1,1,0]
     tagDictionary [{tagName=Athena.AI.Token.Range.Melee}]
    }
   }
  }
  tests [
   EnvQueryTest_Distance_0 {
    class=EnvQueryTest_Distance
    properties {
     distanceTo=&EnvQueryContext_Querier
     filterType=Range
     floatValueMin=0
     scoringEquation=Linear
     scoringFactor=-1
     testComment=""
     testMode=Distance3D
     testPurpose=Score
     referenceValue {dataBinding=null, dataField=FloatValue, defaultValue=0}
    }
   }
   FortQueryTest_InsideAIBotLeash_0 {class=FortQueryTest_InsideAIBotLeash, properties {testPurpose=FilterAndScore}}
  ]
 }]
