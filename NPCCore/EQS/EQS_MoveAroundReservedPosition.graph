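# EQS query that picks a spot for an NPC to move to around its token's reserved position.
# An OnCircle generator places a small ring of projected points around that reserved position, with tracing disabled.
# A 3D distance test keeps points within range of the querier and linearly favors farther ones.
# A half-weight random filter-and-score test varies the chosen point.
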EnvQuery /NPCCore/EQS/EQS_MoveAroundReservedPosition
options [{
  generator {
   name=EnvQueryGenerator_OnCircle_0
   class=EnvQueryGenerator_OnCircle
   properties {
    circleCenter=/Script/FortniteAIServer.FortQueryContext_TokenReservedPosition
    circleRadius=300
    numberOfPoints=12
    projectionData {extentX=0, navigationFilter=null, postProjectionVerticalOffset=0, projectDown=1024}
    traceData {navigationFilter=null, traceMode=None}
   }
  }
  tests [
   FortQueryTest_Random_0 {class=FortQueryTest_Random, properties {floatValueMax=1, scoringFactor=0.5, testPurpose=FilterAndScore}}
   EnvQueryTest_Distance_0 {
    class=EnvQueryTest_Distance
    properties {
     distanceTo=&EnvQueryContext_Querier
     filterType=Range
     floatValueMax=800
     floatValueMin=0
     scoringEquation=Linear
     scoringFactor=1
     testComment=""
     testMode=Distance3D
     testPurpose=FilterAndScore
    }
   }
  ]
 }]
