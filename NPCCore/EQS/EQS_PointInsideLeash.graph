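# EQS query that picks a point inside an AI bot's leash area.
# An OnCircle generator places a ring of candidate points around the closest location inside the leash, projecting them downward with tracing disabled.
# A single InsideAIBotLeash test both filters and scores those candidates.
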EnvQuery /NPCCore/EQS/EQS_PointInsideLeash
options [{
  generator {
   name=EnvQueryGenerator_OnCircle_0
   class=EnvQueryGenerator_OnCircle
   properties {
    circleCenter=/Script/FortniteAIServer.FortQueryContext_ClosestLocationInsideLeash
    circleRadius=1000
    numberOfPoints=16
    projectionData {extentX=0, navigationFilter=null, postProjectionVerticalOffset=0, projectDown=1024}
    traceData {navigationFilter=null, traceMode=None}
   }
  }
  tests [FortQueryTest_InsideAIBotLeash_0 {class=FortQueryTest_InsideAIBotLeash, properties {testPurpose=FilterAndScore}}]
 }]
